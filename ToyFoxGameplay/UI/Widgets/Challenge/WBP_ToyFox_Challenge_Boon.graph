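# ToyFox challenge widget showing personal damage inflicted as a progress bar beside a Megazord boon icon.
# UpdateDamageDealtPercent, zeroed on Construct, drives the bar's dynamic material scalar "Progress 02" and stores the value for GetDamageDealtPercent.
# The visible bar is a material-driven Image, while the stock ProgressBar is collapsed.
# Anim_BoonRewarded spins and pulses the icon, then fades out the whole widget; no graph here plays it.
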WidgetBlueprint /ToyFoxGameplay/UI/Widgets/Challenge/WBP_ToyFox_Challenge_Boon
animations [Anim_BoonRewarded {
  binding_count=2
  length_seconds=0.75
  object=&self:Anim_BoonRewarded
  sequence {bindings [
    Image_Boon {
     id=#1
     possessed_class=Image
     tracks [{
       type=MovieScene2DTransformTrack
       sections [{
         type=MovieScene2DTransformSection
         end_frame=10
         start_frame=0
         channels [
          Translation.X {type=MovieSceneScriptingFloatChannel, const [0,1]}
          Translation.Y {type=MovieSceneScriptingFloatChannel, const [0,1]}
          Angle {type=MovieSceneScriptingFloatChannel, thinned_from=3, keys [[0,0],[10,360]]}
          Scale.X {type=MovieSceneScriptingFloatChannel, keys [[0,1],[5,1.5],[10,1]]}
          Scale.Y {type=MovieSceneScriptingFloatChannel, keys [[0,1],[5,1.5],[10,1]]}
          Shear.X {type=MovieSceneScriptingFloatChannel, const [0,1]}
          Shear.Y {type=MovieSceneScriptingFloatChannel, const [0,1]}
         ]
        }]
      }]
    }
    VerticalBox_Root {
     id=#2
     possessed_class=VerticalBox
     tracks [{
       type=MovieSceneFloatTrack
       display_name=Render Opacity
       sections [{type=MovieSceneFloatSection, end_frame=15, start_frame=0, channels [None {type=MovieSceneScriptingFloatChannel, keys [[0,1],[10,1],[15,0]]}]}]
      }]
    }
   ]}
 }]
graphs [
 EventGraph {nodes [
   Event Construct()
   > UpdateDamageDealtPercent!(execute, self, Progress:float=0) -> (then)
  ]}
 GetDamageDealtPercent {nodes [
   GetDamageDealtPercent() {is=title K2Node_FunctionEntry}
   > Return Node!(execute, Percent:float=Get ProgressFloat() -> (ProgressFloat:float))
  ]}
 UpdateDamageDealtPercent {nodes [
   n1 = UpdateDamageDealtPercent!() -> (then, Progress:float=~) {is=title K2Node_FunctionEntry}
   > n2 = GetDynamicMaterial(self=Get Image_DamageBar_BG() -> (Image_DamageBar_BG:Image Object Reference))
   > SetScalarParameterValue(self=@n2, ParameterName=Progress 02, Value=@n1)
   > Set ProgressFloat(ProgressFloat:float=@n1) -> (Output_Get:float)
  ]}
]
variables [ProgressFloat {type=number}]
widgets {
 info {inheritedWidgetCount=0, parentClass=&UserWidget, rootWidgetClass=&VerticalBox, widgetCount=8}
 widgets [
  VerticalBox_Root {widgetClassPath=&VerticalBox, widget=VerticalBox_Root}
  Text_PersonalDamage {
   widgetClassPath=&CommonTextBlock
   parent=VerticalBox_Root
   widget=Text_PersonalDamage
   properties {
    justification=Center
    mobileFontSizeMultiplier=1.3
    text=Damage Inflicted
    textTransformPolicy=ToUpper
    font {
     bForceMonospaced=false
     bMaterialIsStencil=false
     fontMaterial=null
     fontObject=&BurbankSmall
     letterSpacing=0
     monospacedWidth=1
     size=10.5
     skewAmount=0
     typefaceFontName=Black
     outlineSettings {
      bApplyOutlineToDropShadows=false
      bMiteredCorners=false
      bSeparateFillAlpha=false
      outlineMaterial=&M-FontShadow
      outlineSize=3
      outlineColor=<rgba 0 0 0 0.5>
     }
    }
   }
   slot {class=VerticalBoxSlot, properties {horizontalAlignment=HAlign_Center, verticalAlignment=VAlign_Bottom, padding=<ltrb 0 0 0 -10>}}
  }
  HBox_BoonBar {widgetClassPath=&HorizontalBox, bIsVariable=true, parent=VerticalBox_Root, widget=HBox_BoonBar, slot {class=VerticalBoxSlot}}
  SizeBox_DamageDealtBar {
   widgetClassPath=&SizeBox
   bIsVariable=true
   parent=HBox_BoonBar
   widget=SizeBox_DamageDealtBar
   properties {bOverride_HeightOverride=true, bOverride_WidthOverride=true, heightOverride=25, minDesiredWidth=200, widthOverride=400}
   slot {class=HorizontalBoxSlot, properties {verticalAlignment=VAlign_Center}}
  }
  Overlay {widgetClassPath=&Overlay, parent=SizeBox_DamageDealtBar, widget=Overlay, slot {class=SizeBoxSlot}}
  Image_DamageBar_BG {
   widgetClassPath=&Image
   bIsVariable=true
   parent=Overlay
   widget=Image_DamageBar_BG
   properties {brush {
     bIsDynamicallyLoaded=false
     drawAs=Image
     imageType=FullColor
     mirroring=NoMirror
     resourceName=None
     resourceObject=/ToyFoxGameplay/UI/Materials/MedallionBoonsProgress/MI_UI_Boon_Progress
     tiling=NoTile
     imageSize=<32 32>
     margin=<ltrb 0 0 0 0>
     outlineSettings {
      bUseBrushTransparency=false
      roundingType=HalfHeightRadius
      width=0
      color {colorUseRule=UseColor_Specified, specifiedColor=<rgba 0 0 0 0>}
      cornerRadii=<0 0 0 0>
     }
     tintColor {colorUseRule=UseColor_Specified, specifiedColor=<rgba 1 1 1 1>}
     uVRegion {bIsValid=false, max=<0 0>, min=<0 0>}
    }}
   slot {class=OverlaySlot, properties {horizontalAlignment=HAlign_Fill, verticalAlignment=VAlign_Fill}}
  }
  ProgressBar {
   widgetClassPath=/Script/UMG.ProgressBar
   bIsVariable=true
   parent=Overlay
   widget=ProgressBar
   properties {
    percent=0.25
    visibility=Collapsed
    borderPadding=<2 2>
    fillColorAndOpacity=<rgba 0.0324 0.1629 1 1>
    widgetStyle {
     enableFillAnimation=false
     backgroundImage {
      bIsDynamicallyLoaded=false
      drawAs=Box
      imageType=FullColor
      mirroring=NoMirror
      resourceName=../../../Engine/Content/Slate/Common/ProgressBar_Background.png
      resourceObject=null
      tiling=NoTile
      imageSize=<0 0>
      margin=<ltrb 0.4167 0.4167 0.4167 0.4167>
      outlineSettings {
       bUseBrushTransparency=false
       roundingType=HalfHeightRadius
       width=0
       color {colorUseRule=UseColor_Specified, specifiedColor=<rgba 0 0 0 0>}
       cornerRadii=<0 0 0 0>
      }
      tintColor {colorUseRule=UseColor_Specified, specifiedColor=<rgba 1 1 1 1>}
      uVRegion {bIsValid=false, max=<0 0>, min=<0 0>}
     }
     fillImage {
      bIsDynamicallyLoaded=false
      drawAs=Box
      imageType=FullColor
      mirroring=NoMirror
      resourceName=../../../Engine/Content/Slate/Common/ProgressBar_Fill.png
      resourceObject=null
      tiling=NoTile
      imageSize=<0 0>
      margin=<ltrb 0.4167 0.4167 0.4167 0.4167>
      outlineSettings {
       bUseBrushTransparency=false
       roundingType=HalfHeightRadius
       width=0
       color {colorUseRule=UseColor_Specified, specifiedColor=<rgba 0 0 0 0>}
       cornerRadii=<0 0 0 0>
      }
      tintColor {colorUseRule=UseColor_Specified, specifiedColor=<rgba 1 1 1 1>}
      uVRegion {bIsValid=false, max=<0 0>, min=<0 0>}
     }
     marqueeImage {
      bIsDynamicallyLoaded=false
      drawAs=Image
      imageType=FullColor
      mirroring=NoMirror
      resourceName=../../../Engine/Content/Slate/Common/ProgressBar_Marquee.png
      resourceObject=null
      tiling=Horizontal
      imageSize=<20 12>
      margin=<ltrb 0 0 0 0>
      outlineSettings {
       bUseBrushTransparency=false
       roundingType=HalfHeightRadius
       width=0
       color {colorUseRule=UseColor_Specified, specifiedColor=<rgba 0 0 0 0>}
       cornerRadii=<0 0 0 0>
      }
      tintColor {colorUseRule=UseColor_Specified, specifiedColor=<rgba 1 1 1 1>}
      uVRegion {bIsValid=false, max=<0 0>, min=<0 0>}
     }
    }
   }
   slot {class=OverlaySlot, properties {horizontalAlignment=HAlign_Fill, verticalAlignment=VAlign_Fill}}
  }
  Image_Boon {
   widgetClassPath=&Image
   parent=HBox_BoonBar
   widget=Image_Boon
   properties {brush {
     bIsDynamicallyLoaded=false
     drawAs=Image
     imageType=FullColor
     mirroring=NoMirror
     resourceName=None
     resourceObject=/ToyFoxGameplay/UI/Materials/Textures/Megazord_BoonIcon
     tiling=NoTile
     imageSize=<40 40>
     margin=<ltrb 0 0 0 0>
     outlineSettings {
      bUseBrushTransparency=false
      roundingType=HalfHeightRadius
      width=0
      color {colorUseRule=UseColor_Specified, specifiedColor=<rgba 0 0 0 0>}
      cornerRadii=<0 0 0 0>
     }
     tintColor {colorUseRule=UseColor_Specified, specifiedColor=<rgba 1 1 1 1>}
     uVRegion {bIsValid=false, max=<0 0>, min=<0 0>}
    }}
   slot {class=HorizontalBoxSlot, properties {horizontalAlignment=HAlign_Right, verticalAlignment=VAlign_Center}}
  }
 ]
}
