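# UI material that draws a two-value progress bar for the boon medallion, using alpha-composite blending.
# An anti-aliased rounded-box SDF, inset by Padding, forms the background mask.
# Progress 01 and Progress 02 step against horizontal UV to form fill masks.
# Background and two tint colors combine through these masks into a saturated color, while weighted masks sum into a saturated alpha.
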Material /ToyFoxGameplay/UI/Materials/MedallionBoonsProgress/M_UI_BoonMedallion_Progress
graph {
 properties {bAutomaticallySetUsageInEditor=false, blendMode=BLEND_AlphaComposite, materialDomain=MD_UI, usedShadingModels=Default Lit}
 nodes [
  n1 = ComponentMask(None=TextureCoordinate()) {r=true}
  n2 = ScalarParameter() {defaultValue=0.8, desc=Default 0, group=Progress, parameterName=Progress 02, sliderMax=1, sortPriority=3}
  NamedRerouteDeclaration(None=Multiply(A=Step(Y=@n1, X=@n2) {constX=0.9, constY=1}, B=NamedRerouteUsage())) {name=Progress_02, nodeColor=<rgba 0 0 0 1>}
  n3 = ScalarParameter() {defaultValue=0.5, parameterName=Padding}
  n4 = MaterialFunctionCall(Float2=MaterialFunctionCall() {desc=Uvs, materialFunction=&GetUserInterfaceUV}.Pixel Size) {materialFunction=&BreakOutFloat2Components}
  n5 = Divide(A=@n4.G, B=@n4.G)
  n6 = OneMinus(None=AppendVector(A=Multiply(A=@n3, B=@n5), B=@n3))
  n7 = Constant() {r=0.1}
  n8 = StaticBool() {value=true}
  NamedRerouteDeclaration(None=MaterialFunctionCall(DistanceField=MaterialFunctionCall!(UVs, Size=@n6, CornerRoundness=@n7, KeepAspectRatio=@n8, PixelSize) {materialFunction=/FortUILibrary/MaterialFunctions/Shapes/SDFs/MF_SDF_Box}.SDF) {desc=BarBGMask, materialFunction=&MF_UI_AntiAliasing}.Inner) {name=Bg, nodeColor=<rgba 0 0 0 1>}
  n9 = ComponentMask(None=TextureCoordinate()) {r=true}
  n10 = ScalarParameter() {defaultValue=0.5, desc=Default 0, group=Progress, parameterName=Progress 01, sliderMax=1, sortPriority=3}
  NamedRerouteDeclaration(None=Multiply(A=Step(Y=@n9, X=@n10) {constX=0.9, constY=1}, B=NamedRerouteUsage())) {name=Progress 01, nodeColor=<rgba 0 0 0 1>}
  n11 = Multiply(A=VectorParameter() {parameterName=Bg Color, defaultValue=<rgba 0.0208 0 0.0003 1>}.RGB, B=Multiply(A=NamedRerouteUsage()) {constB=0.7})
  n12 = Multiply(A=VectorParameter() {parameterName=Color 01, defaultValue=<rgba 0.8438 0.2922 0 1>}.RGB, B=Multiply(A=NamedRerouteUsage()))
  n13 = VectorParameter() {parameterName=Color 02, defaultValue=<rgba 1 0.0819 0 1>}
  n14 = Multiply(A=NamedRerouteUsage())
  Saturate(None=LinearInterpolate(A=Add(A=@n11, B=@n12), B=Multiply(A=@n13.RGB, B=@n14), Alpha=@n14))
  n15 = Multiply(A=NamedRerouteUsage()) {constB=0.2}
  n16 = Multiply(A=NamedRerouteUsage()) {constB=0.2}
  n17 = NamedRerouteUsage()
  Saturate(None=Add(A=Add(A=@n15, B=@n16), B=Multiply(A=@n17) {constB=0.7}))
 ]
}
