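# UI material for the medallion boons progress display, pairing the ToyFox medallion icon with an animated pulse ring.
# Looping time, or an override time, shaped by a falloff power drives the ring's scale and opacity.
# Two smoothsteps on an exponential radial gradient cut the soft-edged band, published as the "Ring" reroute.
# The center-scaled icon's squared color is lerped over a dimmed orange layer fed by a reroute, using icon alpha.
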Material /ToyFoxGameplay/UI/Materials/MedallionBoonsProgress/M_UI_Medallion
graph {
 properties {bAutomaticallySetUsageInEditor=false, blendMode=BLEND_AlphaComposite, materialDomain=MD_UI, usedShadingModels=Default Lit}
 nodes [
  n1 = ScalarParameter() {defaultValue=0.01, parameterName=EdgeSoftness}
  n2 = LinearInterpolate(A=ScalarParameter() {defaultValue=1, parameterName=OuterStartScale}, B=ScalarParameter() {parameterName=OuterEndScale})
  Clamp(None=MaterialFunctionCall(Min=@n2, Max=Add(A=@n2, B=@n1)) {materialFunction=&SmoothStep}.Result)
  n3 = NamedRerouteUsage()
  n4 = TextureSample(UVs=MaterialFunctionCall(Texture Scale=Constant() {r=0.4}) {materialFunction=&ScaleUVsByCenter}.UVs) {texture=/TidalCraneMedallionGameplay/Assets/Icons/T_UI_Medallion_ToyFox}
  n5 = NamedRerouteUsage()
  LinearInterpolate(
   A=Multiply(A=Multiply(A=@n3) {constB=0.2}, B=Constant3Vector() {constant=<rgba 1 0.2232 0 1>})
   B=Multiply(A=@n4.RGB, B=@n4.RGB)
   Alpha=Add(A=Multiply(A=@n4.A), B=Multiply(A=@n5) {constB=0})
  ) {constAlpha=1}
  n6 = ScalarParameter() {defaultValue=0.01, parameterName=EdgeSoftness}
  n7 = Multiply(A=Time(), B=ScalarParameter() {defaultValue=1, parameterName=PulseSpeed})
  n8 = ScalarParameter() {parameterName=OverrideTime}
  n9 = ScalarParameter() {parameterName=UseSetTime}
  n10 = Power(
   Base=Fmod(A=LinearInterpolate(A=@n7, B=@n8, Alpha=@n9), B=Constant() {r=1})
   Exp=ScalarParameter() {defaultValue=2, parameterName=FalloffPower}
  )
  n11 = LinearInterpolate(A=ScalarParameter() {defaultValue=1, parameterName=OuterStartScale}, B=ScalarParameter() {parameterName=OuterEndScale}, Alpha=@n10)
  n12 = MaterialFunctionCall(
   UVs=MaterialFunctionCall() {materialFunction=&GetUserInterfaceUV}.9-Slice UV
   Radius=ScalarParameter() {defaultValue=0.5, parameterName=DefaultRadius}
   Density=ScalarParameter() {defaultValue=2.33, parameterName=DefaultDensity}
  ) {materialFunction=&RadialGradientExponential}
  n13 = Clamp(None=MaterialFunctionCall(Alpha=@n12.RadialGradientExponential, Min=@n11, Max=Add(A=@n11, B=@n6)) {materialFunction=&SmoothStep}.Result)
  n14 = Add(
   A=@n11
   B=LinearInterpolate(
    A=ScalarParameter() {defaultValue=0.1, parameterName=InnerStartScale}
    B=ScalarParameter() {defaultValue=0.01, parameterName=InnerEndScale}
    Alpha=@n10
   )
  )
  n15 = Add(A=@n14, B=ScalarParameter() {defaultValue=0.01, parameterName=EdgeSoftness})
  n16 = OneMinus(None=Clamp(None=MaterialFunctionCall(Alpha=@n12.RadialGradientExponential, Min=@n14, Max=@n15) {materialFunction=&SmoothStep}.Result))
  n17 = ScalarParameter() {defaultValue=1, parameterName=StartOpacity}
  n18 = ScalarParameter() {parameterName=EndOpacity}
  NamedRerouteDeclaration(None=Multiply(A=Multiply(A=@n13, B=@n16), B=LinearInterpolate(A=@n17, B=@n18, Alpha=@n10) {constA=1, constB=0})) {name=Ring, nodeColor=<rgba 0 1 0.2734 1>}
 ]
}
