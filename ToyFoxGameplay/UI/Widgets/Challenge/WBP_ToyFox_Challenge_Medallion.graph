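# Medallion challenge HUD widget that shows the local player's and an opponent's damage progress on a shared bar beside a medallion.
# Construct zeroes both values through two update functions.
# Each sets its own scalar on the bar's dynamic material and slides its label by lerped render translation, skipping both at max damage.
# The opposing update also hides the opponent marker at zero damage and labels it #1 or #2 by IsMostDamage.
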WidgetBlueprint /ToyFoxGameplay/UI/Widgets/Challenge/WBP_ToyFox_Challenge_Medallion
graphs [
 EventGraph {nodes [
   Event Construct()
   > UpdateLocalProgress!(execute, self, LocalPlayerDamage:float=0, AtMaxDamage:bool=false) -> (then)
   > UpdateOpposingDamage!(execute, self, OpposingPlayerDamage:float=0, IsMostDamage:bool=false, AtMaxDamage:bool=false) -> (then)
  ]}
 UpdateLocalProgress {nodes [
   UpdateLocalProgress!() -> (then, LocalPlayerDamage:float=~, AtMaxDamage:bool=~) {is=title K2Node_FunctionEntry}
   > n1 = Branch(Condition=Get AtMaxDamage() -> (AtMaxDamage:bool))
   n2 = GetDynamicMaterial(execute=@n1.else, self=Get Image_Progress() -> (Image_Progress:Image Object Reference))
   > SetScalarParameterValue(self=@n2, ParameterName=Progress 01, Value=Get LocalPlayerDamage() -> (LocalPlayerDamage:float))
   > SetRenderTranslation!(
    execute
    self=Get Text_LocalPlayer() -> (Text_LocalPlayer:Common Text Object Reference)
    Translation_X:double=Lerp(A=-200, B=200, Alpha=Get LocalPlayerDamage())
    Translation_Y:double=0
   ) -> (then)
  ]}
 UpdateOpposingDamage {nodes [
   n1 = UpdateOpposingDamage!() -> (then, OpposingPlayerDamage:float=~, IsMostDamage:bool=~, AtMaxDamage:bool=~) {is=title K2Node_FunctionEntry}
   > SetVisibility(
    self=Get Overlay_OpposingIcon() -> (Overlay_OpposingIcon:Overlay Object Reference)
    InVisibility=Select(Option 0=Hidden, Option 1=HitTestInvisible, Index="float > float"(A=@n1))
   )
   > n2 = Branch(Condition=Get AtMaxDamage())
   n3 = GetDynamicMaterial(execute=@n2.else, self=Get Image_Progress())
   > SetScalarParameterValue(self=@n3, ParameterName=Progress 02, Value=Get OpposingPlayerDamage() -> (OpposingPlayerDamage:float))
   > "SetText (Text)"(
    self=Get Text_OpponentRank() -> (Text_OpponentRank:Common Text Object Reference)
    InText=SelectText(A="LOCTEXT(\"##1\")", B="LOCTEXT(\"##2\")", bPickA=Get IsMostDamage() -> (IsMostDamage:bool))
   )
   > SetRenderTranslation(self=Get Text_OpponentRank(), Translation_X=Lerp(A=-200, B=200, Alpha=Get OpposingPlayerDamage()), Translation_Y=0)
  ]}
]
variables []
widgets {
 info {inheritedWidgetCount=0, parentClass=/Script/FortniteUI.FortUserWidget, rootWidgetClass=&HorizontalBox, widgetCount=16}
 widgets [
  HorizontalBox_0 {widgetClassPath=&HorizontalBox, widget=HorizontalBox_0}
  VerticalBox_Root {widgetClassPath=&VerticalBox, parent=HorizontalBox_0, widget=VerticalBox_Root, slot {class=HorizontalBoxSlot}}
  Overlay_0 {
   widgetClassPath=&Overlay
   parent=VerticalBox_Root
   widget=Overlay_0
   slot {class=VerticalBoxSlot, properties {padding=<ltrb 0 0 0 -10>}}
  }
  Text_PersonalDamage {
   widgetClassPath=&CommonTextBlock
   parent=Overlay_0
   widget=Text_PersonalDamage
   properties {
    justification=Center
    text=Damage Inflicted
    textTransformPolicy=ToUpper
    font {
     bForceMonospaced=false
     bMaterialIsStencil=false
     fontMaterial=null
     fontObject=&BurbankSmall
     letterSpacing=0
     monospacedWidth=1
     size=10.5
     skewAmount=0
     typefaceFontName=Black
     outlineSettings {
      bApplyOutlineToDropShadows=false
      bMiteredCorners=false
      bSeparateFillAlpha=false
      outlineMaterial=&M-FontShadow
      outlineSize=3
      outlineColor=<rgba 0 0 0 0.5>
     }
    }
   }
   slot {class=OverlaySlot, properties {horizontalAlignment=HAlign_Center, verticalAlignment=VAlign_Bottom}}
  }
  HorizontalBox_1 {widgetClassPath=&HorizontalBox, parent=VerticalBox_Root, widget=HorizontalBox_1, slot {class=VerticalBoxSlot}}
  SizeBox_0 {
   widgetClassPath=&SizeBox
   parent=HorizontalBox_1
   widget=SizeBox_0
   properties {bOverride_HeightOverride=true, bOverride_WidthOverride=true, heightOverride=25, widthOverride=400}
   slot {class=HorizontalBoxSlot, properties {verticalAlignment=VAlign_Center}}
  }
  Overlay_ProgressBar {
   widgetClassPath=&Overlay
   bIsVariable=true
   parent=SizeBox_0
   widget=Overlay_ProgressBar
   slot {class=SizeBoxSlot, properties {verticalAlignment=VAlign_Center}}
  }
  SizeBox_ProgressBar {
   widgetClassPath=&SizeBox
   bIsVariable=true
   parent=Overlay_ProgressBar
   widget=SizeBox_ProgressBar
   properties {bOverride_HeightOverride=true, bOverride_WidthOverride=true, heightOverride=25, widthOverride=400}
   slot {class=OverlaySlot, properties {horizontalAlignment=HAlign_Fill, verticalAlignment=VAlign_Center}}
  }
  Overlay_70 {widgetClassPath=&Overlay, parent=SizeBox_ProgressBar, widget=Overlay_70, slot {class=SizeBoxSlot}}
  Overlay_LocalPlayer {
   widgetClassPath=&Overlay
   parent=Overlay_70
   widget=Overlay_LocalPlayer
   slot {class=OverlaySlot, properties {horizontalAlignment=HAlign_Fill, verticalAlignment=VAlign_Fill}}
  }
  Image_Progress {
   widgetClassPath=&Image
   bIsVariable=true
   parent=Overlay_LocalPlayer
   widget=Image_Progress
   properties {brush {
     bIsDynamicallyLoaded=false
     drawAs=Image
     imageType=FullColor
     mirroring=NoMirror
     resourceName=None
     resourceObject=/ToyFoxGameplay/UI/Materials/MedallionBoonsProgress/M_UI_BoonMedallion_Progress
     tiling=NoTile
     imageSize=<32 32>
     margin=<ltrb 0 0 0 0>
     outlineSettings {
      bUseBrushTransparency=false
      roundingType=HalfHeightRadius
      width=0
      color {colorUseRule=UseColor_Specified, specifiedColor=<rgba 0 0 0 0>}
      cornerRadii=<0 0 0 0>
     }
     tintColor {colorUseRule=UseColor_Specified, specifiedColor=<rgba 1 1 1 1>}
     uVRegion {bIsValid=false, max=<0 0>, min=<0 0>}
    }}
   slot {class=OverlaySlot, properties {horizontalAlignment=HAlign_Fill, verticalAlignment=VAlign_Fill}}
  }
  Text_LocalPlayer {
   widgetClassPath=&CommonTextBlock
   bIsVariable=true
   parent=Overlay_LocalPlayer
   widget=Text_LocalPlayer
   properties {
    justification=Center
    text=You
    textTransformPolicy=ToUpper
    font {
     bForceMonospaced=false
     bMaterialIsStencil=false
     fontMaterial=null
     fontObject=&BurbankSmall
     letterSpacing=36
     monospacedWidth=0.552
     size=12
     skewAmount=0
     typefaceFontName=Bold
     outlineSettings {
      bApplyOutlineToDropShadows=false
      bMiteredCorners=false
      bSeparateFillAlpha=false
      outlineMaterial=null
      outlineSize=0
      outlineColor=<rgba 0.0781 0.0781 0.0781 0.235>
     }
    }
    renderTransform {angle=0, scale=<1 1>, shear=<0 0>, translation=<-200 0>}
    renderTransformPivot=<0.5 0>
    shadowColorAndOpacity=<rgba 0 0 0 0.56>
    shadowOffset=<1.5 1.5>
   }
   slot {class=OverlaySlot, properties {horizontalAlignment=HAlign_Center, verticalAlignment=VAlign_Center, padding=<ltrb 0 0 0 -24>}}
  }
  Overlay_OpposingIcon {
   widgetClassPath=&Overlay
   bIsVariable=true
   parent=Overlay_70
   widget=Overlay_OpposingIcon
   slot {class=OverlaySlot, properties {horizontalAlignment=HAlign_Fill, verticalAlignment=VAlign_Fill}}
  }
  Text_OpponentRank {
   widgetClassPath=&CommonTextBlock
   bIsVariable=true
   parent=Overlay_OpposingIcon
   widget=Text_OpponentRank
   properties {
    justification=Center
    text="##2"
    textTransformPolicy=ToUpper
    font {
     bForceMonospaced=false
     bMaterialIsStencil=false
     fontMaterial=null
     fontObject=&BurbankSmall
     letterSpacing=36
     monospacedWidth=0.552
     size=12
     skewAmount=0
     typefaceFontName=Medium
     outlineSettings {
      bApplyOutlineToDropShadows=false
      bMiteredCorners=false
      bSeparateFillAlpha=false
      outlineMaterial=null
      outlineSize=0
      outlineColor=<rgba 0.0781 0.0781 0.0781 0.235>
     }
    }
    renderTransform {angle=0, scale=<1 1>, shear=<0 0>, translation=<-200 0>}
    renderTransformPivot=<0.5 0>
    shadowColorAndOpacity=<rgba 0 0 0 0.56>
    shadowOffset=<1.5 1.5>
   }
   slot {class=OverlaySlot, properties {horizontalAlignment=HAlign_Center, verticalAlignment=VAlign_Center, padding=<ltrb 0 0 0 15>}}
  }
  Overlay_Medallion {widgetClassPath=&Overlay, parent=HorizontalBox_1, widget=Overlay_Medallion, slot {class=HorizontalBoxSlot}}
  Image_MedallionPulse {
   widgetClassPath=&Image
   parent=Overlay_Medallion
   widget=Image_MedallionPulse
   properties {
    brush {
     bIsDynamicallyLoaded=false
     drawAs=Image
     imageType=FullColor
     mirroring=NoMirror
     resourceName=None
     resourceObject=/ToyFoxGameplay/UI/Materials/MedallionBoonsProgress/M_UI_Medallion
     tiling=NoTile
     imageSize=<40 40>
     margin=<ltrb 0 0 0 0>
     outlineSettings {
      bUseBrushTransparency=false
      roundingType=HalfHeightRadius
      width=0
      color {colorUseRule=UseColor_Specified, specifiedColor=<rgba 0 0 0 0>}
      cornerRadii=<0 0 0 0>
     }
     tintColor {colorUseRule=UseColor_Specified, specifiedColor=<rgba 1 1 1 1>}
     uVRegion {bIsValid=false, max=<0 0>, min=<0 0>}
    }
    renderTransform {angle=0, scale=<2 2>, shear=<0 0>, translation=<0 0>}
   }
   slot {class=OverlaySlot, properties {horizontalAlignment=HAlign_Center, verticalAlignment=VAlign_Center}}
  }
 ]
}
