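# UI material for the Zord Health boon progress bar, drawn as a segmented pink-to-orange gradient bar with alpha-composite blending.
# Skewed SDF boxes form the base masks, while mirrored MSDF textures shape the top and bottom bars, all anti-aliased through shared UI functions.
# A repeating horizontal UV fraction creates the segments, and the scalar Progress parameter moves a thresholded horizontal fill mask.
# Named reroutes carry these masks into the layered color and opacity math.
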Material /ToyFoxGameplay/UI/Materials/MedallionBoonsProgress/M_UI_Boon_ZordHealth
graph {
 properties {bAutomaticallySetUsageInEditor=false, blendMode=BLEND_AlphaComposite, materialDomain=MD_UI, usedShadingModels=Default Lit}
 nodes [
  n1 = Add(A=NamedRerouteUsage(), B=NamedRerouteUsage())
  n2 = NamedRerouteUsage()
  n3 = NamedRerouteUsage()
  n4 = NamedRerouteUsage()
  Add(A=Add(A=Multiply(A=@n1) {constB=0.1}, B=Multiply(A=@n2) {constB=0.5}), B=Add(A=Multiply(A=@n3) {constB=0.3}, B=Multiply(A=@n4) {constB=0.3}))
  n5 = Multiply(A=Constant3Vector() {constant=<rgba 0.0182 0.0016 0.0103 1>}, B=Multiply(A=NamedRerouteUsage()) {constB=0.6}) {constB=0}
  n6 = Multiply(A=Constant3Vector() {constant=<rgba 0.0859 0 0.0105 1>}, B=Multiply(A=NamedRerouteUsage()) {constB=0.5})
  n7 = Multiply(
   A=LinearInterpolate(A=Constant3Vector() {constant=<rgba 1 0 0.0903 1>}, B=Constant3Vector() {constant=<rgba 1 0.4573 0 1>}, Alpha=NamedRerouteUsage())
   B=NamedRerouteUsage()
  )
  n8 = NamedRerouteUsage()
  n9 = NamedRerouteUsage()
  Add(A=LinearInterpolate(A=Add(A=@n5, B=@n6), B=@n7, Alpha=@n7), B=Multiply(A=Add(A=@n8, B=@n9)) {constB=0.1})
  MaterialFunctionCall(Speed=ScalarParameter() {defaultValue=0.05, desc=Test, parameterName=DebugTimeSine}) {materialFunction=&DebugTimeSine}
  n10 = Add(A=TextureCoordinate(), B=Constant2Vector() {r=0.005})
  n11 = Frac(None=Multiply(A=ComponentMask(None=@n10) {r=true}) {constB=38})
  n12 = Constant() {r=0.3}
  NamedRerouteDeclaration(None=Multiply(
    A=MaterialFunctionCall(DistanceField=@n11, Threshold=@n12) {materialFunction=&MF_UI_AntiAliasing}.Outer
    B=NamedRerouteUsage()
   )) {name=Segments, nodeColor=<rgba 0 0 0 1>}
  n13 = Constant2Vector() {g=1, r=8}
  n14 = MaterialFunctionCall(Float2=MaterialFunctionCall!(
    UV=MaterialFunctionCall(TextureSize=@n13) {materialFunction=&MF_UI_UVScaling}.Scaled to Fit
    Mirror Y
    Swizzle UV
   ) {materialFunction=/FortUILibrary/MaterialFunctions/Basics/MF_UI_Symmetry_2x1_New}.Flipped Symmetry) {materialFunction=&BreakOutFloat2Components}
  n15 = TextureSample(UVs=AppendVector(A=Subtract(A=@n14.R) {constB=0.25}, B=Subtract(A=@n14.G) {constB=0.2})) {texture=&/T_UI_BoonHealth_Bottom}
  NamedRerouteDeclaration(None=MaterialFunctionCall(
    DistanceField=MaterialFunctionCall!(MSDF RGB=@n15.RGB) {materialFunction=/FortUILibrary/MaterialFunctions/Shapes/Modifiers/MF_UI_MSDFtoSDF}.SDF
    Threshold=Constant() {r=0.3}
    Smoothing=Constant()
   ) {materialFunction=&MF_UI_AntiAliasing}.Outer) {name=Bottom Bar, nodeColor=<rgba 0 0 0 1>}
  n16 = MaterialFunctionCall(Float2=MaterialFunctionCall(UV=MaterialFunctionCall() {materialFunction=&MF_UI_UVScaling}.Scaled to Fit) {materialFunction=&MF_UI_Symmetry_2x1_New}.Symmetry) {materialFunction=&BreakOutFloat2Components}
  n17 = MaterialFunctionCall!(UV=AppendVector(A=Subtract(A=@n16.R) {constB=0.2}, B=@n16.G), Skew Angle=Constant2Vector() {r=-50}) {materialFunction=/FortUILibrary/MaterialFunctions/Basics/MF_UI_Skew}
  n18 = Constant2Vector() {g=0.32, r=16.5}
  NamedRerouteDeclaration(None=MaterialFunctionCall(DistanceField=MaterialFunctionCall!(UVs=@n17.Skewed UVs, Size=@n18, CornerRoundness, KeepAspectRatio, PixelSize) {materialFunction=/FortUILibrary/MaterialFunctions/Shapes/SDFs/MF_SDF_Box}.SDF) {materialFunction=&MF_UI_AntiAliasing}.Inner) {name=Skewed Base Inner Mask, nodeColor=<rgba 0 0 0 1>}
  n19 = Constant2Vector() {g=0.8, r=7}
  n20 = MaterialFunctionCall(Float2=MaterialFunctionCall(UV=MaterialFunctionCall(TextureSize=@n19) {materialFunction=&MF_UI_UVScaling}.Scaled to Fit) {materialFunction=&MF_UI_Symmetry_2x1_New}.Flipped Symmetry) {materialFunction=&BreakOutFloat2Components}
  n21 = Subtract(A=@n20.R) {constB=0.15}
  n22 = Subtract(A=@n20.G) {constB=-0.3}
  n23 = TextureSample(UVs=MaterialFunctionCall(UVs=AppendVector(A=@n21, B=@n22), "Rotation Angle (0-1)"=Constant() {r=-0.01}) {materialFunction=&CustomRotator}.Rotated Values) {texture=&/T_UI_BoonHealth_Top}
  NamedRerouteDeclaration(None=MaterialFunctionCall(
    DistanceField=MaterialFunctionCall(MSDF RGB=@n23.RGB) {materialFunction=&MF_UI_MSDFtoSDF}.SDF
    Threshold=Constant() {r=0.5}
    Smoothing=Constant()
   ) {materialFunction=&MF_UI_AntiAliasing}.Outer) {name=Top Bar, nodeColor=<rgba 0 0 0 1>}
  n24 = Add(A=@n18) {constB=0.12}
  NamedRerouteDeclaration(None=MaterialFunctionCall(DistanceField=MaterialFunctionCall(UVs=@n17.Skewed UVs, Size=@n24) {materialFunction=&MF_SDF_Box}.SDF) {materialFunction=&MF_UI_AntiAliasing}.Inner) {name=Skewed Base Outer Mask, nodeColor=<rgba 0 0 0 1>}
  n25 = MaterialFunctionCall(Float2=TextureCoordinate()) {materialFunction=&BreakOutFloat2Components}
  NamedRerouteDeclaration(None=@n25.R) {name=R, nodeColor=<rgba 0 0 0 1>}
  NamedRerouteDeclaration(None=@n25.G) {name=G, nodeColor=<rgba 0 0 0 1>}
  n26 = TextureCoordinate()
  n27 = Constant() {desc=Progress Start, r=0.35}
  n28 = Constant() {desc=Progress End, r=1.25}
  n29 = ScalarParameter() {defaultValue=1, desc=Default 0, group=Progress, parameterName=Progress, sliderMax=1, sortPriority=3}
  n30 = ComponentMask(None=Add(A=Multiply(A=@n26) {constB=-1}, B=LinearInterpolate(A=@n27, B=@n28, Alpha=@n29)) {constB=0.5}) {r=true}
  n31 = Constant() {r=0.3}
  NamedRerouteDeclaration(None=Multiply(
    A=NamedRerouteUsage()
    B=MaterialFunctionCall(DistanceField=@n30, Threshold=@n31) {materialFunction=&MF_UI_AntiAliasing}.Outer
   )) {name=Progress, nodeColor=<rgba 0 0 0 1>}
 ]
}
