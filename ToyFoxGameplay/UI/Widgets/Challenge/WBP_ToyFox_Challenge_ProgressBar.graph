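# ToyFox challenge HUD widget that shows a "MEGAZORD" progress bar, apparently for Zord health judging by its material name.
# A centered text label sits above a fixed-size image that draws the bar with the M_UI_Boon_ZordHealth material.
# The UpdateProgress event clamps its input to 0–1 and writes it to the Progress scalar on the image's dynamic material, so the material renders the fill.
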WidgetBlueprint /ToyFoxGameplay/UI/Widgets/Challenge/WBP_ToyFox_Challenge_ProgressBar
graphs [EventGraph {nodes [
   n1 = UpdateProgress!() -> (OutputDelegate, then, Progress:double=~) {is=title K2Node_CustomEvent}
   > n2 = GetDynamicMaterial(self=Get Image_Progress() -> (Image_Progress:Image Object Reference))
   > SetScalarParameterValue(self=@n2, ParameterName=Progress, Value="Clamp (Float)"(Value=@n1.Progress, Min=0, Max=1))
  ]}]
variables []
widgets {
 info {inheritedWidgetCount=0, parentClass=/Script/FortniteUI.FortUserWidget, rootWidgetClass=&Overlay, widgetCount=5}
 widgets [
  Overlay_23 {widgetClassPath=&Overlay, widget=Overlay_23}
  VerticalBox_28 {widgetClassPath=&VerticalBox, parent=Overlay_23, widget=VerticalBox_28, slot {class=OverlaySlot}}
  CommonTextBlock_160 {
   widgetClassPath=&CommonTextBlock
   parent=VerticalBox_28
   widget=CommonTextBlock_160
   properties {
    text=MEGAZORD
    font {
     bForceMonospaced=false
     bMaterialIsStencil=false
     fontMaterial=null
     fontObject=&BurbankSmall
     letterSpacing=0
     monospacedWidth=1
     size=24
     skewAmount=0
     typefaceFontName=Black
     outlineSettings {
      bApplyOutlineToDropShadows=false
      bMiteredCorners=false
      bSeparateFillAlpha=false
      outlineMaterial=&M-FontShadow
      outlineSize=3
      outlineColor=<rgba 0 0 0 0.5>
     }
    }
    renderTransform {angle=0, scale=<1 1>, shear=<0 0>, translation=<0 10>}
    shadowOffset=<0 0>
   }
   slot {class=VerticalBoxSlot, properties {horizontalAlignment=HAlign_Center}}
  }
  SizeBox_0 {
   widgetClassPath=&SizeBox
   parent=VerticalBox_28
   widget=SizeBox_0
   properties {bOverride_HeightOverride=true, bOverride_WidthOverride=true, heightOverride=80, widthOverride=800}
   slot {class=VerticalBoxSlot}
  }
  Image_Progress {
   widgetClassPath=&Image
   bIsVariable=true
   parent=SizeBox_0
   widget=Image_Progress
   properties {brush {
     bIsDynamicallyLoaded=false
     drawAs=Image
     imageType=FullColor
     mirroring=NoMirror
     resourceName=None
     resourceObject=/ToyFoxGameplay/UI/Materials/MedallionBoonsProgress/M_UI_Boon_ZordHealth
     tiling=NoTile
     imageSize=<32 32>
     margin=<ltrb 0 0 0 0>
     outlineSettings {
      bUseBrushTransparency=false
      roundingType=HalfHeightRadius
      width=0
      color {colorUseRule=UseColor_Specified, specifiedColor=<rgba 0 0 0 0>}
      cornerRadii=<0 0 0 0>
     }
     tintColor {colorUseRule=UseColor_Specified, specifiedColor=<rgba 1 1 1 1>}
     uVRegion {bIsValid=false, max=<0 0>, min=<0 0>}
    }}
   slot {class=SizeBoxSlot}
  }
 ]
}
