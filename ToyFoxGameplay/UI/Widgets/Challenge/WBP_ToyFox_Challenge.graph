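# Challenge HUD widget for the ToyFox giant event, showing giant power level and the view target player's damage toward rewards.
# Binds to game state component events to drive the progress bar, boon and medallion child widgets.
# A completed boon plays its reward animation and gives way to the hotfix-gated medallion, comparing player damage against top rivals.
# A looping timer, started by event state, toggles visibility by distance to the giant.
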WidgetBlueprint /ToyFoxGameplay/UI/Widgets/Challenge/WBP_ToyFox_Challenge
graphs [
 EventGraph {nodes [
   n1 = HandleGiantPowerLevelChanged!() -> (OutputDelegate, then, PowerLevel:double=~) {is=title K2Node_CustomEvent}
   > UpdateProgress!(
    execute
    self:WBP Toy Fox Challenge Progress Bar Object Reference=Get WBP_ToyFox_Challenge_ProgressBar() -> (WBP_ToyFox_Challenge_ProgressBar:WBP Toy Fox Challenge Progress Bar Object Reference)
    Progress:double=@n1.PowerLevel
   ) -> (then)
   HandlePlayerDamageTrackingInfoUpdated() {is=title K2Node_CustomEvent}
   > Branch(Condition=Get bWidgetVisible() -> (bWidgetVisible:bool))
   > n2 = Branch(Condition=Get IsBoonActive() -> (IsBoonActive:bool))
   > n3 = UpdateBoonWidgetData!(execute, self) -> (then, bBoonJustCompleted:bool)
   > Branch(Condition=@n3)
   > Set IsBoonActive(IsBoonActive:bool=false) -> (Output_Get:bool)
   n4 = Get WBP_ToyFox_Challenge_Boon() -> (WBP_ToyFox_Challenge_Boon:WBP Toy Fox Challenge Boon Object Reference)
   > n5 = Play Animation with Finished event!(
    execute:Exec
    Widget:User Widget Object Reference=@n4
    InAnimation:Widget Animation Object Reference=Get Anim_BoonRewarded!(self:WBP Toy Fox Challenge Boon Object Reference=@n4) -> (Anim_BoonRewarded:Widget Animation Object Reference)
    StartAtTime:float=0
    NumLoopsToPlay:int=1
    PlayMode:EUMGSequencePlayMode Enum
    PlaybackSpeed:float=1
   ) -> (then:Exec, Result:Widget Animation Handle Structure, Finished:Exec) {is=title K2Node_PlayAnimation2}
   > Set BoonRewardedAnimationHandle(BoonRewardedAnimationHandle:Widget Animation Handle Structure=@n5) -> (Output_Get:Widget Animation Handle Structure)
   > ShowMedallionUI()
   > UpdateMedallionWidgetData()
   SetVisibility(execute=@n5.Finished, self=@n4, InVisibility=Collapsed)
   UpdateMedallionWidgetData(execute=@n2.else)
   n6 = HandleViewTargetChanged!() -> (
    OutputDelegate
    then
    PlayerController:Fort Player Controller Object Reference
    OldViewTarget:Actor Object Reference
    NewViewTarget:Actor Object Reference
   ) {is=title K2Node_CustomEvent}
   > FortShippingLog!(
    execute
    WorldContextObject:Object Reference=Self-Reference()
    InString:String=Append!(
     A="HandleViewTargetChanged: For player controller "
     B=GetDisplayName(Object=@n6.PlayerController)
     C:String=" from "
     D:String=GetDisplayName(Object=@n6.OldViewTarget)
     E:String=" to "
     F:String=GetDisplayName(Object=@n6.NewViewTarget)
    ) -> (ReturnValue)
    bLogAsWarning:bool=false
   ) -> (then)
   > UpdateWidgetVisibility()
   > n7 = Sequence()
   > Branch(Condition=Get bWidgetVisible())
   > StopAllAnimations(self:User Widget Object Reference=[
     Get WBP_ToyFox_Challenge_Boon()
     Get WBP_ToyFox_Challenge_Medallion() -> (WBP_ToyFox_Challenge_Medallion:WBP Toy Fox Challenge Medallion Object Reference)
    ])
   n8 = Get View Target as Player Pawn!(In:Exec=@n7.then_1) -> (Success:Exec, Fail:Exec, PlayerPawn:Fort Player Pawn Object Reference) {is=title K2Node_MacroInstance}
   > n9 = Get!(execute:Exec) -> (then:Exec, else:Exec, GameStateComponent:B Toy Fox Game State Component Object Reference)
   > n10 = IsPlayerActiveRewardGoalBoon!(
    execute
    self:B Toy Fox Game State Component Object Reference=@n9
    Player:Fort Player State Athena Object Reference=Cast To FortPlayerStateAthena!(Object=Get PlayerState!(self:Pawn Object Reference=@n8) -> (PlayerState:Player State Object Reference)) -> (AsFort Player State Athena:Fort Player State Athena Object Reference, bSuccess:bool=~) {is=title K2Node_DynamicCast}
   ) -> (then, ReturnValue:bool)
   > SetBoonActive!(execute, self, IsActive:bool=@n10) -> (then)
   Event Construct()
   > Parent: Construct()
   > FortShippingLog(WorldContextObject=Self-Reference(), InString=Construct: Invoked, bLogAsWarning=false)
   > SetVisibility(InVisibility=Collapsed)
   > n11 = Sequence()
   n12 = GetGameStateAthena!() -> (ReturnValue:Fort Game State Athena Object Reference)
   > n13 = Is Valid(InputObject=@n12)
   FortShippingLog(execute=@n13.Is Not Valid, WorldContextObject=Self-Reference(), InString=Construct: GameState is null., bLogAsWarning=true)
   n14 = GetComponentByClass!(
    execute=@n13.Is Valid
    self:Actor Object Reference=@n12
    ComponentClass:Actor Component Class Reference=/ToyFoxGameplay/Gameplay/B_ToyFoxGameStateComponent.B_ToyFoxGameStateComponent_C
   ) -> (then, ReturnValue:B Toy Fox Game State Component Object Reference)
   > n15 = Is Valid(InputObject=@n14)
   FortShippingLog(
    execute=@n15.Is Not Valid
    WorldContextObject=Self-Reference()
    InString=Construct: GameStateComponent is found. Binding to it.
    bLogAsWarning=false
   )
   n16 = Set GameStateComponent(execute=@n15.Is Valid, GameStateComponent:B Toy Fox Game State Component Object Reference=@n14) -> (Output_Get:B Toy Fox Game State Component Object Reference)
   > FortShippingLog(WorldContextObject=Self-Reference(), InString=Construct: GameStateComponent is found. Binding to it., bLogAsWarning=false)
   > HandleEventStateChanged!(
    execute
    self
    EventState:ToyFoxEventStateEnum Enum=GetEventState!(self:B Toy Fox Game State Component Object Reference=@n16) -> (State:ToyFoxEventStateEnum Enum=NewEnumerator0)
   ) -> (then)
   > Bind Event to On Event State Changed(self:B Toy Fox Game State Component Object Reference=@n16, Delegate=Create Event()) {is=title K2Node_AddDelegate}
   > Bind Event to On Giant Power Level Changed(self:B Toy Fox Game State Component Object Reference=@n16, Delegate=Create Event()) {is=title K2Node_AddDelegate}
   > Bind Event to On Player Damage Tracking Info Updated(self:B Toy Fox Game State Component Object Reference=@n16, Delegate=Create Event()) {is=title K2Node_AddDelegate}
   > HandleGiantPowerLevelChanged!(execute, self, PowerLevel:double=GetPowerLevel!(self:B Toy Fox Game State Component Object Reference=@n16) -> (Return:double)) -> (then) {is=title K2Node_CallFunction}
   n17 = Cast To FortPlayerController(execute=@n11.then_1, Object=GetOwningPlayer())
   > Bind Event to On View Target Changed(self:Fort Player Controller Object Reference=@n17, Delegate=Create Event()) {is=title K2Node_AddDelegate}
   > HandleViewTargetChanged!(
    execute
    self
    PlayerController:Fort Player Controller Object Reference
    OldViewTarget:Actor Object Reference
    NewViewTarget:Actor Object Reference=GetViewTarget!(self:Controller Object Reference=@n17) -> (ReturnValue:Actor Object Reference)
   ) -> (then) {is=title K2Node_CallFunction}
  ]}
 GetViewTargetAsPlayerPawn {nodes [
   Inputs!() -> (In:Exec)
   > n1 = Cast To FortPlayerController(Object=GetOwningPlayer())
   > n2 = Cast To FortPlayerPawn(Object=GetViewTarget(self=@n1))
   > Outputs!(Success:Exec, Fail:Exec=[@n2.CastFailed, @n1.CastFailed], PlayerPawn:Fort Player Pawn Object Reference=@n2)
  ]}
 HandleEventStateChanged {nodes [
   n1 = HandleEventStateChanged!() -> (then, EventState:ToyFoxEventStateEnum Enum=NewEnumerator0) {is=title K2Node_FunctionEntry}
   > Branch(Condition="Equal (Enum)"(A:ToyFoxEventStateEnum Enum=@n1, B:ToyFoxEventStateEnum Enum=NewEnumerator3))
   > FortShippingLog(
    WorldContextObject=Self-Reference()
    InString=Append!(A="HandleEventStateChanged: Event state changed to  ", B=Enum to String!(Enumerator:ToyFoxEventStateEnum Enum=@n1) -> (ReturnValue:String)) -> (ReturnValue)
    bLogAsWarning=false
   )
   n2 = Get UpdateWidgetTimerHandle() -> (UpdateWidgetTimerHandle:Timer Handle Structure)
   > Branch(Condition=NOT Boolean(A=Is Timer Active by Handle!(Handle:Timer Handle Structure=@n2) -> (ReturnValue:bool)))
   > n3 = Set Timer by Event(Delegate=Create Event(), Time=0.5, bLooping=true, bMaxOncePerFrame=true)
   > Set UpdateWidgetTimerHandle(UpdateWidgetTimerHandle:Timer Handle Structure=@n3) -> (Output_Get:Timer Handle Structure)
   > UpdateWidgetVisibility()
  ]}
 IsMedallionUIEnabled {nodes [
   IsMedallionUIEnabled() {is=title K2Node_FunctionEntry}
   n1 = Get Hotfix_ShowMedallionUI() -> (Hotfix_ShowMedallionUI:Scalable Float Structure)
   > Return Node!(
    execute
    IsEnabled:bool="float >= float"(A=GetValueAtLevel!(Input:"Scalable Float Structure (by ref)"=@n1, Level:float=0) -> (ReturnValue:float), B=0.999)
   )
  ]}
 SetBoonActive {nodes [
   n1 = SetBoonActive!() -> (then, IsActive:bool=~) {is=title K2Node_FunctionEntry}
   > n2 = Set IsBoonActive(IsBoonActive=@n1)
   > n3 = Branch(Condition=@n2)
   > SetVisibility(self=Get WBP_ToyFox_Challenge_Boon(), InVisibility=SelfHitTestInvisible)
   > SetVisibility(self=Get WBP_ToyFox_Challenge_Medallion(), InVisibility=Collapsed)
   > UpdateBoonWidgetData()
   SetVisibility(execute=@n3.else, self=Get WBP_ToyFox_Challenge_Boon(), InVisibility=Collapsed)
   > ShowMedallionUI()
   > UpdateMedallionWidgetData()
  ]}
 ShowMedallionUI {nodes [
   ShowMedallionUI() {is=title K2Node_FunctionEntry}
   > Branch(Condition=IsMedallionUIEnabled!(self) -> (IsEnabled:bool) {is=title K2Node_CallFunction})
   > SetVisibility(self=Get WBP_ToyFox_Challenge_Medallion(), InVisibility=SelfHitTestInvisible)
  ]}
 UpdateBoonWidgetData {nodes [
   UpdateBoonWidgetData() {is=title K2Node_FunctionEntry}
   > n1 = Sequence()
   > n2 = Get View Target as Player Pawn()
   > n3 = Cast To FortPlayerStateAthena!(execute:Exec, Object=Get PlayerState(self=@n2)) -> (then:Exec, CastFailed:Exec, AsFort Player State Athena)
   > n4 = Get()
   > n5 = GetDamageDoneToGiantByPlayer!(execute, self:B Toy Fox Game State Component Object Reference=@n4, Player:Fort Player State Athena Object Reference=@n3) -> (then, Damage:double)
   n6 = "Clamp (Float)"(
    Value=SafeDivide(A=@n5, B=GetRewardBoonThresholdDamage!(self:B Toy Fox Game State Component Object Reference=@n4) -> (ReturnValue:float))
    Min=0
    Max=1
   )
   > Set NewDamagePercent(NewDamagePercent:double=@n6) -> (Output_Get:double)
   n7 = Get WBP_ToyFox_Challenge_Boon()
   > Set OldDamagePercent(OldDamagePercent:float=GetDamageDealtPercent!(self:WBP Toy Fox Challenge Boon Object Reference=@n7) -> (Percent:float)) -> (Output_Get:float)
   > UpdateDamageDealtPercent!(execute, self:WBP Toy Fox Challenge Boon Object Reference=@n7, Progress:float=@n6) -> (then)
   n8 = Get NewDamagePercent() -> (NewDamagePercent:double)
   n9 = Get NewDamagePercent()
   n10 = Get OldDamagePercent() -> (OldDamagePercent:float)
   > Return Node!(execute, bBoonJustCompleted:bool=AND Boolean(A="Nearly Equal (Float)"(A=@n8, B=1, ErrorTolerance=0), B="float < float"(A=@n10, B=@n9)))
   Return Node(execute=@n1.then_1, bBoonJustCompleted=false)
  ]}
 UpdateMedallionWidgetData {nodes [
   UpdateMedallionWidgetData() {is=title K2Node_FunctionEntry}
   > Branch(Condition=IsMedallionUIEnabled())
   > n1 = Get View Target as Player Pawn()
   > n2 = Cast To FortPlayerStateAthena(Object=Get PlayerState(self=@n1))
   > n3 = Get()
   > n4 = GetDamageDoneToGiantByPlayer(self=@n3, Player=@n2)
   > Set ViewTargetPlayerDamage(ViewTargetPlayerDamage:double=@n4) -> (Output_Get:double)
   > n5 = GetHighestDamageDoneToGiantByAnyPlayer!(execute, self:B Toy Fox Game State Component Object Reference=@n3, ExcludingPlayerStates:Array of Fort Player State Object References) -> (then, DamageInfo:Toy Fox Player Damage Struct Structure)
   > Set bViewTargetPlayerHighestDamage(bViewTargetPlayerHighestDamage:bool="Equal (Object)"(
     A=Break Toy Fox Player Damage Struct!(ToyFoxPlayerDamageStruct:"Toy Fox Player Damage Struct Structure (by ref)"=@n5) -> (Player_2_#1:Fort Player State Object Reference, TotalDamage_5_#2:double=~) {is=title K2Node_BreakStruct}
     B=@n3
    )) -> (Output_Get:bool)
   n6 = GetCachedScalableFloatValue!(Input:"Fort Cached Scalable Float Structure (by ref)"=Get Hotfix_MaxDamagePossible() -> (Hotfix_MaxDamagePossible:Fort Cached Scalable Float Structure)) -> (ReturnValue:float)
   n7 = Get ViewTargetPlayerDamage() -> (ViewTargetPlayerDamage:double)
   > UpdateLocalProgress!(
    execute
    self:WBP Toy Fox Challenge Medallion Object Reference=Get WBP_ToyFox_Challenge_Medallion()
    LocalPlayerDamage:float=NormalizeToRange!(Value:double=SafeDivide(A=@n7, B=@n6), RangeMin:double=0, RangeMax:double=1) -> (ReturnValue:double)
    AtMaxDamage:bool="float >= float"(A=@n7, B=@n6)
   ) -> (then)
   > n8 = Branch(Condition=Get bViewTargetPlayerHighestDamage() -> (bViewTargetPlayerHighestDamage:bool))
   > n9 = GetHighestDamageDoneToGiantByAnyPlayer!(
    execute
    self=Get GameStateComponent() -> (GameStateComponent:B Toy Fox Game State Component Object Reference)
    ExcludingPlayerStates=Make Array("[0]":Fort Player State Object Reference=@n2) -> (Array:Array of Fort Player State Object References)
   ) -> (
    then
    DamageInfo_Player_2_#1:Fort Player State Object Reference
    DamageInfo_TotalDamage_5_#2:double=~
   )
   n10 = GetCachedScalableFloatValue(Input=Get Hotfix_MaxDamagePossible())
   > UpdateOpposingDamage!(
    execute
    self:WBP Toy Fox Challenge Medallion Object Reference=Get WBP_ToyFox_Challenge_Medallion()
    OpposingPlayerDamage:float=NormalizeToRange(Value=SafeDivide(A=@n9.DamageInfo_TotalDamage_5_#2, B=@n10), RangeMin=0, RangeMax=1)
    IsMostDamage:bool=false
    AtMaxDamage:bool="float >= float"(A=@n9.DamageInfo_TotalDamage_5_#2, B=@n10)
   ) -> (then)
   n11 = GetHighestDamageDoneToGiantByAnyPlayer(execute=@n8.else, self=Get GameStateComponent()) -> (DamageInfo_TotalDamage_5_#2=~)
   n12 = GetCachedScalableFloatValue(Input=Get Hotfix_MaxDamagePossible())
   > UpdateOpposingDamage(
    self=Get WBP_ToyFox_Challenge_Medallion()
    OpposingPlayerDamage=NormalizeToRange(Value=SafeDivide(A=@n11.DamageInfo_TotalDamage_5_#2, B=@n12), RangeMin=0, RangeMax=1)
    IsMostDamage=true
    AtMaxDamage="float >= float"(A=@n11.DamageInfo_TotalDamage_5_#2, B=@n12)
   )
  ]}
 UpdateWidgetVisibility {nodes [
   UpdateWidgetVisibility() {is=title K2Node_FunctionEntry}
   > Branch(Condition=IsInPrioSwitcher!(self:Fort Prioritized Widget Interface) -> (ReturnValue:bool))
   > n1 = Sequence()
   > n2 = Get()
   n3 = GetGiantPawn!(self:B Toy Fox Game State Component Object Reference=@n2) -> (Pawn:Toy Fox Giant Player Pawn Object Reference)
   > Is Valid(InputObject=@n3)
   n4 = GetCachedScalableFloatValue(Input=Get Hotfix_MinViewableDistance() -> (Hotfix_MinViewableDistance:Fort Cached Scalable Float Structure))
   > Set NewVisibility(NewVisibility:bool="float < float"(
     A=GetSquaredHorizontalDistanceTo!(self:Actor Object Reference=@n3, OtherActor:Actor Object Reference) -> (ReturnValue:float)
     B=Square!(A:double=@n4) -> (ReturnValue:double)
    )) -> (Output_Get:bool)
   n5 = Delta Gate Wildcard!(execute:Exec=@n1.then_1, Variable:"Boolean (by ref)"=Get bWidgetVisible(), New Value:bool=Get NewVisibility() -> (NewVisibility:bool)) -> (Success:Exec, Updated Variable:"Boolean (by ref)", Previous Variable:bool=~) {is=title K2Node_MacroInstance}
   > n6 = Sequence()
   > n7 = Branch(Condition=@n5)
   > SetVisibility(InVisibility=HitTestInvisible)
   SetVisibility(execute=@n7.else, InVisibility=Collapsed)
   n8 = GetVisibility!(self:Widget Object Reference) -> (ReturnValue:ESlateVisibility Enum=Visible)
   FortShippingLog(
    execute=@n6.then_1
    WorldContextObject=Self-Reference()
    InString=Append(A="UpdateWidgetVisibility: Updating visibility to ", B=Enum to String(Enumerator:ESlateVisibility Enum=@n8))
    bLogAsWarning=false
   )
  ]}
]
variables [
 BoonRewardedAnimationHandle {type=WidgetAnimationHandle, default {}}
 bWidgetVisible {type=boolean}
 GameStateComponent {type=B_ToyFoxGameStateComponent_C, default=null}
 Hotfix_MaxDamagePossible {
  type=FortCachedScalableFloat
  default {scalableFloat {
    registryType=None
    value=1
    curve {curveTable=/ToyFoxGameplay/DataTables/ToyFoxGameplayGameData, rowName=ToyFox.Event.Reward.Medallion.MaxDamagePossible}
   }}
 }
 Hotfix_MinViewableDistance {type=FortCachedScalableFloat, default {scalableFloat {registryType=None, value=28000, curve {curveTable=null, rowName=None}}}}
 Hotfix_ShowMedallionUI {
  type=ScalableFloat
  default {registryType=None, value=1, curve {curveTable=&ToyFoxGameplayGameData, rowName=ToyFox.UI.Challenge.ShowMedallion}}
 }
 IsBoonActive {type=boolean}
 UpdateWidgetTimerHandle {type=TimerHandle}
]
widgets {
 info {inheritedWidgetCount=1, parentClass=/Script/FortniteUI.AthenaCommonPrioritizedWidget, rootWidgetClass=&VerticalBox, widgetCount=5}
 widgets [
  VerticalBox_47 {widgetClassPath=&VerticalBox, widget=VerticalBox_47}
  WBP_ToyFox_Challenge_ProgressBar {
   widgetClassPath=&/WBP_ToyFox_Challenge_ProgressBar
   bIsVariable=true
   parent=VerticalBox_47
   widget=WBP_ToyFox_Challenge_ProgressBar
   slot {class=VerticalBoxSlot}
  }
  WBP_ToyFox_Challenge_Boon {
   widgetClassPath=&/WBP_ToyFox_Challenge_Boon
   bIsVariable=true
   parent=VerticalBox_47
   widget=WBP_ToyFox_Challenge_Boon
   slot {class=VerticalBoxSlot, properties {horizontalAlignment=HAlign_Center, verticalAlignment=VAlign_Center}}
  }
  WBP_ToyFox_Challenge_Medallion {
   widgetClassPath=&/WBP_ToyFox_Challenge_Medallion
   bIsVariable=true
   parent=VerticalBox_47
   widget=WBP_ToyFox_Challenge_Medallion
   properties {visibility=Collapsed}
   slot {class=VerticalBoxSlot, properties {horizontalAlignment=HAlign_Center}}
  }
  TutorialHighlight {widgetClassPath=&TutorialHighlight}
 ]
}
